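# Additive unlit material for the Landmass erosion brush that, by its name, applies an erosion weightmap.
# It multiplies the red channel of the WeightmapRT texture parameter by an optional mask.
# The mask lerps from 1 to the brush edge falloff of LandmassBrush_ShapeMask, driven by the UseMask scalar, so the shape restriction can be switched off.
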Material /Landmass/Landscape/ErosionBrush/Materials/Weightmap/Erosion_ApplyWeightmap
graph {
 properties {blendMode=BLEND_Additive, pixelDepthOffsetMode=PDOM_Legacy, shadingModel=MSM_Unlit, usedShadingModels=Unlit}
 nodes [
  n1 = TextureSampleParameter2D() {
   parameterName=WeightmapRT
   samplerType=SAMPLERTYPE_LinearColor
   texture=/Engine/Functions/Engine_MaterialFunctions02/PivotPainter2/Black_1x1_EXR_Texture
  }
  Multiply(A=@n1.R, B=NamedRerouteUsage())
  NamedRerouteDeclaration(None=LinearInterpolate(
    B=MaterialFunctionCall!(Falloff Width Override, Falloff Offset Override) {materialFunction=/Landmass/Landscape/BlueprintBrushes/Materials/BrushMaterials/BasicFunctions/LandmassBrush_ShapeMask}.Brush Edge Falloff
    Alpha=ScalarParameter() {parameterName=UseMask}
   ) {constA=1}) {name=Optional Mask, nodeColor=<rgba 0 1 0.1094 1>}
 ]
}
